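# Physics asset for the F_MED_Valentine_Fashion female body cosmetic, covering the ruffle bones on its shorts or short skirt.
# Small capsule bodies sit on five dyn_short_ruffle bones, two left and three right, alongside capsules on both thighs and spine_02.
# Each ruffle is constrained to its same-side thigh with limited swing and a zero-degree twist limit.
# The spine capsule has no constraint.
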PhysicsAsset /BRCosmetics/Characters/Player/Female/Medium/Bodies/F_MED_Valentine_Fashion/Meshes/F_MED_Valentine_Fashion_Physics
bodies {
 dyn_short_ruffle_01_l [{shapeType=Capsule, length=2.25, radius=1.25, shapeName=dyn_short_ruffle_01_l_capsule, center=<1.1559 -0.0001 0>}]
 dyn_short_ruffle_01_r [{shapeType=Capsule, length=2.25, radius=1.25, shapeName=dyn_short_ruffle_01_r_capsule, center=<1.1559 -0.0001 0>}]
 dyn_short_ruffle_02_l [{shapeType=Capsule, length=2.25, radius=1.25, shapeName=dyn_short_ruffle_02_l_capsule, center=<1.1559 -0.0001 0>}]
 dyn_short_ruffle_02_r [{shapeType=Capsule, length=2.25, radius=1.25, shapeName=dyn_short_ruffle_02_r_capsule, center=<1.1559 -0.0001 0>}]
 dyn_short_ruffle_03_r [{shapeType=Capsule, length=2.25, radius=1.25, shapeName=dyn_short_ruffle_03_r_capsule, center=<1.1559 -0.0001 0>}]
 spine_02 [{shapeType=Capsule, length=3, radius=3, shapeName=spine_02_capsule}]
 thigh_l [{shapeType=Capsule, length=2, radius=2}]
 thigh_r [{shapeType=Capsule, length=2, radius=2}]
}
constraints [
 {bone1Name=dyn_short_ruffle_01_l, bone2Name=thigh_l, swing1LimitDegrees=30, swing2LimitDegrees=30, twistLimitDegrees=0}
 {bone1Name=dyn_short_ruffle_02_l, bone2Name=thigh_l, swing1LimitDegrees=30, swing2LimitDegrees=30, twistLimitDegrees=0}
 {bone1Name=dyn_short_ruffle_01_r, bone2Name=thigh_r, swing1LimitDegrees=30, swing2LimitDegrees=30, twistLimitDegrees=0}
 {bone1Name=dyn_short_ruffle_02_r, bone2Name=thigh_r, swing1LimitDegrees=30, swing2LimitDegrees=30, twistLimitDegrees=0}
 {bone1Name=dyn_short_ruffle_03_r, bone2Name=thigh_r, swing1LimitDegrees=30, swing2LimitDegrees=30, twistLimitDegrees=0}
]
