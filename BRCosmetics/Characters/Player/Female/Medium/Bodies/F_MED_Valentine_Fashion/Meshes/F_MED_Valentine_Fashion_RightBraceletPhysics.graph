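# Physics asset for the right bracelet on the F_MED_Valentine_Fashion female player body, providing the bodies and joint for bracelet physics.
# A large capsule on the dyn_bracelet_r_1 bone pairs with a very small capsule on lowerarm_r, which serves as its anchor.
# A single constraint links the two bones with narrow, symmetric swing limits and a zero twist limit, restricting the bracelet to slight wobble around the forearm.
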PhysicsAsset /BRCosmetics/Characters/Player/Female/Medium/Bodies/F_MED_Valentine_Fashion/Meshes/F_MED_Valentine_Fashion_RightBraceletPhysics
bodies {
 dyn_bracelet_r_1 [{shapeType=Capsule, length=3.9766, radius=3.945, shapeName=dyn_bracelet_r_1_capsule, center=<0.7914 0 0>, rotation=<rot 90 0 0>}]
 lowerarm_r [{shapeType=Capsule, length=0.505, radius=0.505, shapeName=lowerarm_r_capsule, rotation=<rot 0 -90 0>}]
}
constraints [{bone1Name=dyn_bracelet_r_1, bone2Name=lowerarm_r, swing1LimitDegrees=10, swing2LimitDegrees=10, twistLimitDegrees=0}]
