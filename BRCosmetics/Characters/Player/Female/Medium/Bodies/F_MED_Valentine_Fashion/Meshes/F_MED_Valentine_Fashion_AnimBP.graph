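# Animation Blueprint for the F_MED_Valentine_Fashion female body, adding deformation and secondary motion to the parent mesh's copied pose.
# Twist and sculpter Control Rigs feed body and right-bracelet RigidBody sims, a bracelet bone offset, then belt AnimDynamics.
# Each update, right upperarm rotation against spine_05 sets that offset's alpha, shifting the bracelet as the arm moves.
# A bracelet AnimDynamics node is left unconnected.
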AnimBlueprint /BRCosmetics/Characters/Player/Female/Medium/Bodies/F_MED_Valentine_Fashion/Meshes/F_MED_Valentine_Fashion_AnimBP
target_skeleton=&/F_MED_Valentine_Fashion_Skeleton
graphs [
 AnimGraph {nodes [
   Anim Dynamics - Bone: dyn_bracelet_r_1(Alpha=1, ExternalForce="0,0,0") {
    bUseGravityOverride=true
    simulationSpace=BoneRelative
    alphaBoolBlend {blendOption=Linear}
    boundBone {boneName=dyn_bracelet_r_1}
    componentAppliedLinearAccClamp=<6000 6000 6000>
    componentLinearAccScale=<0.2 0.2 0.2>
    componentLinearVelScale=<0.2 0.2 0.2>
    gravityOverride {z=-980}
    physicsBodyDefinitions [{
      sphereCollisionRadius=0
      boundBone {boneName=dyn_bracelet_r_1}
      boxExtents=<5 5 5>
      constraintSetup {angularLimitsMax=<0 0 0>, angularLimitsMin=<0 0 0>, angularTarget=<0 0 0>, linearAxesMax=<0.6 0.7 0.6>, linearAxesMin=<-0.6 -0.7 0>}
      localJointOffset=<1 -1.5 -1>
     }]
    relativeSpaceBone {boneName=clavicle_r}
   }
   n1 = Control Rig(Source=Copy Pose From Mesh() {bUseAttachedParent=true}) {
    alphaBoolBlend {blendOption=Linear}
    controlRigAssetReference {blueprintRigClass=&/F_MED_Valentine_Fashion_Deform_CtrlRig_twists.F_MED_Valentine_Fashion_Deform_CtrlRig_twists_C}
   }
   n2 = Control Rig!(
    Source=@n1
    Alpha
    KneeRot_fr_l_114deg:Vector
    KneeTrans_fr_l_114deg:Vector="0.258,-3.591,0.009"
    KneeScale_fr_l_114deg:Vector="1,1,1"
    KneeScale_fr_l_130deg:Vector="1,1,1"
    KneeRot_fr_l_130deg:Vector="0,0,-51.2"
    KneeTrans_fr_l_130deg:Vector="0.258,-3.591,0.009"
   ) -> (Pose) {
    alphaBoolBlend {blendOption=Linear}
    controlRigAssetReference {blueprintRigClass=&/F_MED_Valentine_Fashion_Deform_CtrlRig_sculpters.F_MED_Valentine_Fashion_Deform_CtrlRig_sculpters_C}
   }
   n3 = RigidBody(ComponentPose=Local To Component(LocalPose=@n2)) {
    bClampLinearTranslationLimitToRefPose=true
    bOverrideWorldGravity=true
    overridePhysicsAsset=&/F_MED_Valentine_Fashion_Physics
    simulationSpace=BaseBoneSpace
    alphaBoolBlend {blendOption=Linear}
    baseBoneRef {boneName=root}
    componentAppliedLinearAccClamp=<6000 6000 6000>
    componentLinearAccScale=<0.2 0.2 0.2>
    componentLinearVelScale=<0.2 0.2 0.2>
    overrideWorldGravity {z=-980}
   }
   n4 = RigidBody(ComponentPose=@n3) {
    bClampLinearTranslationLimitToRefPose=true
    bOverrideWorldGravity=true
    overridePhysicsAsset=&/F_MED_Valentine_Fashion_RightBraceletPhysics
    simulationSpace=BaseBoneSpace
    alphaBoolBlend {blendOption=Linear}
    baseBoneRef {boneName=lowerarm_r}
    componentAppliedLinearAccClamp=<6000 6000 6000>
    componentLinearAccScale=<0.4 0.4 0.4>
    componentLinearVelScale=<0.4 0.4 0.4>
    overrideWorldGravity {z=-980}
   }
   n5 = "Transform (Modify) Bone - Bone: dyn_bracelet_r_1"(ComponentPose=@n4, Alpha=Get armRt Rot Up() -> (armRt Rot Up:double), Translation="-3,1,0", Rotation="0,0,0", Scale="1,1,1") {
    translationMode=BMM_Additive
    translationSpace=BCS_BoneSpace
    alphaBoolBlend {blendOption=Linear}
    alphaScaleBiasClamp {bMapRange=true}
    boneToModify {boneName=dyn_bracelet_r_1}
    translation=<-3 1>
   }
   n6 = Anim Dynamics - Bone: dyn_belt_01_c(ComponentPose=@n5, Alpha=1, ExternalForce="0,0,0") {
    angularDampingOverride=0.8
    bOverrideLinearDamping=true
    bUseGravityOverride=true
    linearDampingOverride=1
    linearSpringConstant=0.2
    simulationSpace=RootRelative
    alphaBoolBlend {blendOption=Linear}
    boundBone {boneName=dyn_belt_01_c}
    componentAppliedLinearAccClamp=<6000 6000 6000>
    componentLinearAccScale=<0.5 0.5 0.5>
    componentLinearVelScale=<0.5 0.5 0.5>
    gravityOverride {z=-980}
    physicsBodyDefinitions [{
      sphereCollisionRadius=0
      boundBone {boneName=dyn_belt_01_c}
      boxExtents=<3 9 9>
      constraintSetup {angularLimitsMax=<0 0 0>, angularLimitsMin=<0 0 0>, angularTarget=<0 0 0>, linearAxesMax=<3 0.5 3>, linearAxesMin=<-3 -2 -3>}
      localJointOffset=<2 0 0>
     }]
    relativeSpaceBone {boneName=pelvis}
   }
   Output Pose(Result=Component To Local(ComponentPose=@n6)) {layerGroup=None}
  ]}
 EventGraph {nodes [
   Event BlueprintUpdateAnimation()
   > n1 = Sequence()
   > Set body mesh(body mesh:Skeletal Mesh Component Object Reference=GetPartSkeletalMeshComponent(self=Self-Reference())) -> (Output_Get:Skeletal Mesh Component Object Reference)
   Set armRt Rot Up(
    execute=@n1.then_1
    armRt Rot Up:double=GetAlphaFromReferenceTransformDelta(
     SkeletalMeshComponent=Get body mesh() -> (body mesh:Skeletal Mesh Component Object Reference)
     BoneName=upperarm_r
     BaseName=spine_05
     InputRange=(X=100,Y=130)
     TwistAxis="1,0,0"
    )
   ) -> (Output_Get:double)
  ]}
]
variables [
 armRt Rot Up {type=number}
 body mesh {type=SkeletalMeshComponent}
]
