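# Physics asset for the hair and face accessory part of the female medium Valentine Fashion character, simulating ponytail, side hair strands, bangs and ribbons.
# Capsule bodies form jointed chains that hang from the head, with spine, neck and head capsules supplying the body shapes.
# Dynamic joints use swing limits with zero twist, generally loosening toward the tips of hair chains.
# The ponytail root and spine-to-head joints are left free.
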PhysicsAsset /BRCosmetics/Characters/Player/Female/Medium/Bodies/F_MED_Valentine_Fashion/Meshes/Parts/F_MED_Valentine_Fashion_FaceAcc_Physics
bodies {
 dyn_bangs_01 [{shapeType=Capsule, length=0.25, radius=2.25, shapeName=dyn_bangs_01_capsule, center=<0.05 0 0>}]
 dyn_hairMeshBase_l_01 [{shapeType=Capsule, length=3.25, radius=0.75, shapeName=dyn_hairMeshBase_l_01_capsule, center=<2.5 -0.5 0>, rotation=<rot 90 0 0>}]
 dyn_hairMeshBase_r_01 [{shapeType=Capsule, length=3.25, radius=0.75, shapeName=dyn_hairMeshBase_r_01_capsule, center=<2.5 -0.5 0>, rotation=<rot 90 0 0>}]
 dyn_hairMesh_l_01 [{shapeType=Capsule, length=3, radius=1, shapeName=dyn_hairMesh_l_01_capsule, center=<2.5047 0 0.0001>, rotation=<rot 90 0 0>}]
 dyn_hairMesh_l_02 [{shapeType=Capsule, length=3.25, radius=0.75, shapeName=dyn_hairMesh_l_02_capsule, center=<2.5047 0 0.0001>, rotation=<rot 90 0 0>}]
 dyn_hairMesh_l_03 [{shapeType=Capsule, length=3.25, radius=0.75, shapeName=dyn_hairMesh_l_03_capsule, center=<2.5047 0 0.0001>, rotation=<rot 90 0 0>}]
 dyn_hairMesh_r_01 [{shapeType=Capsule, length=3, radius=1, shapeName=dyn_hairMesh_r_01_capsule, center=<2.5047 0 0.0001>, rotation=<rot 90 0 0>}]
 dyn_hairMesh_r_02 [{shapeType=Capsule, length=3.25, radius=0.75, shapeName=dyn_hairMesh_r_02_capsule, center=<2.5047 0 0.0001>, rotation=<rot 90 0 0>}]
 dyn_hairMesh_r_03 [{shapeType=Capsule, length=3.25, radius=0.75, shapeName=dyn_hairMesh_r_03_capsule, center=<2.5047 0 0.0001>, rotation=<rot 90 0 0>}]
 dyn_ponyTailRoot_01 [{shapeType=Capsule, length=1, radius=1, shapeName=dyn_ponyTailRoot_01_capsule, rotation=<rot 90 0 0>}]
 dyn_ponyTail_01 [{shapeType=Capsule, length=0, radius=4.25, shapeName=dyn_ponyTail_01_capsule, center=<3.618 0 0.0001>, rotation=<rot 90 0 0>}]
 dyn_ponyTail_02 [{shapeType=Capsule, length=0.25, radius=5.25, shapeName=dyn_ponyTail_02_capsule, center=<2.5047 0 0.0001>, rotation=<rot 90 0 0>}]
 dyn_ponyTail_03 [{shapeType=Capsule, length=0, radius=5.5, shapeName=dyn_ponyTail_03_capsule, center=<4 1 -0.4>, rotation=<rot 90 0 0>}]
 dyn_ponyTail_04 [{shapeType=Capsule, length=1, radius=3, shapeName=dyn_ponyTail_04_capsule, center=<2.5047 0 0.0001>, rotation=<rot 90 0 0>}]
 dyn_ponyTail_05 [{shapeType=Capsule, length=0.5, radius=2, shapeName=dyn_ponyTail_05_capsule, center=<1.9721 0 0.0001>, rotation=<rot 90 0 0>}]
 dyn_ribbonBase_l [{shapeType=Capsule, length=1, radius=1.5, shapeName=dyn_ribbonBase_l_capsule, center=<0 0 0.5>, rotation=<rot 90 0 0>}]
 dyn_ribbonBase_r [{shapeType=Capsule, length=1, radius=1.5, shapeName=dyn_ribbonBase_r_capsule, center=<0 0 -0.5>, rotation=<rot 90 0 0>}]
 dyn_ribbon_l_01 [{shapeType=Capsule, length=3.25, radius=0.75, shapeName=dyn_ribbon_l_01_capsule, center=<2.5 0 0>, rotation=<rot 90 0 0>}]
 dyn_ribbon_l_02 [{shapeType=Capsule, length=5, radius=0.75, shapeName=dyn_ribbon_l_02_capsule, center=<2.5 0 0>, rotation=<rot 90 0 0>}]
 dyn_ribbon_l_03 [{shapeType=Capsule, length=3.25, radius=0.75, shapeName=dyn_ribbon_l_03_capsule, center=<2.5 0 0>, rotation=<rot 90 0 0>}]
 dyn_ribbon_l_04 [{shapeType=Capsule, length=2.75, radius=0.75, shapeName=dyn_ribbon_l_04_capsule, center=<2 -0.0001 -1>, rotation=<rot 56.1541 0 0>}]
 dyn_ribbon_r_01 [{shapeType=Capsule, length=3.25, radius=0.75, shapeName=dyn_ribbon_r_01_capsule, center=<2.5 0 -0.5>, rotation=<rot 90 0 0>}]
 dyn_ribbon_r_02 [{shapeType=Capsule, length=3.25, radius=0.75, shapeName=dyn_ribbon_r_02_capsule, center=<2.5 0 0>, rotation=<rot 90 0 0>}]
 dyn_ribbon_r_03 [{shapeType=Capsule, length=3, radius=1, shapeName=dyn_ribbon_r_03_capsule, center=<2.5 0 0>, rotation=<rot 90 0 0>}]
 head [
  {shapeType=Capsule, length=2.4401, radius=7.6882, shapeName=head_capsule, center=<6.1427 -1.0075 -0.0001>, rotation=<rot 90 -0.2238 -0.6846>}
  {shapeType=Capsule, length=7.5, radius=6, shapeName=head_capsule, center=<3.2547 1.9422 -0.0001>, rotation=<rot 90 0 6.7222>}
 ]
 neck_01 [{shapeType=Capsule, length=2.75, radius=4.255, shapeName=neck_01_capsule, rotation=<rot 90 0 0>}]
 neck_02 [{shapeType=Capsule, length=2.75, radius=4.255, shapeName=neck_02_capsule, rotation=<rot 90 0 0>}]
 spine_05 [{shapeType=Capsule, length=16, radius=10.005, shapeName=spine_05_capsule, center=<2.1067 1.3879 0>, rotation=<rot 0 -89.9999 0>}]
}
constraints [
 {bone1Name=dyn_ponyTail_01, bone2Name=dyn_ponyTailRoot_01, swing1LimitDegrees=12, swing2LimitDegrees=12, twistLimitDegrees=0}
 {bone1Name=dyn_ponyTail_02, bone2Name=dyn_ponyTail_01, swing1LimitDegrees=15, swing2LimitDegrees=15, twistLimitDegrees=0}
 {bone1Name=dyn_ponyTail_03, bone2Name=dyn_ponyTail_02, swing1LimitDegrees=20, swing2LimitDegrees=20, twistLimitDegrees=0}
 {bone1Name=dyn_ponyTail_04, bone2Name=dyn_ponyTail_03, swing1LimitDegrees=30, swing2LimitDegrees=30, twistLimitDegrees=0}
 {bone1Name=dyn_ponyTail_05, bone2Name=dyn_ponyTail_04, swing1LimitDegrees=30, swing2LimitDegrees=30, twistLimitDegrees=0}
 {bone1Name=dyn_ribbon_l_02, bone2Name=dyn_ribbon_l_01, swing1LimitDegrees=30, swing2LimitDegrees=30, twistLimitDegrees=0}
 {bone1Name=dyn_ribbon_l_03, bone2Name=dyn_ribbon_l_02, swing1LimitDegrees=30, swing2LimitDegrees=30, twistLimitDegrees=0}
 {bone1Name=dyn_ribbon_l_04, bone2Name=dyn_ribbon_l_03, swing1LimitDegrees=30, swing2LimitDegrees=30, twistLimitDegrees=0}
 {bone1Name=dyn_ribbon_r_02, bone2Name=dyn_ribbon_r_01, swing1LimitDegrees=30, swing2LimitDegrees=30, twistLimitDegrees=0}
 {bone1Name=dyn_ribbon_r_03, bone2Name=dyn_ribbon_r_02, swing1LimitDegrees=30, swing2LimitDegrees=30, twistLimitDegrees=0}
 {bone1Name=dyn_hairMesh_r_01, bone2Name=dyn_hairMeshBase_r_01, swing1LimitDegrees=10, swing2LimitDegrees=10, twistLimitDegrees=0}
 {bone1Name=dyn_hairMesh_r_02, bone2Name=dyn_hairMesh_r_01, swing1LimitDegrees=30, swing2LimitDegrees=30, twistLimitDegrees=0}
 {bone1Name=dyn_hairMesh_r_03, bone2Name=dyn_hairMesh_r_02, swing1LimitDegrees=30, swing2LimitDegrees=30, twistLimitDegrees=0}
 {bone1Name=dyn_hairMesh_l_01, bone2Name=dyn_hairMeshBase_l_01, swing1LimitDegrees=10, swing2LimitDegrees=10, twistLimitDegrees=0}
 {bone1Name=dyn_hairMesh_l_02, bone2Name=dyn_hairMesh_l_01, swing1LimitDegrees=30, swing2LimitDegrees=30, twistLimitDegrees=0}
 {bone1Name=dyn_hairMesh_l_03, bone2Name=dyn_hairMesh_l_02, swing1LimitDegrees=30, swing2LimitDegrees=30, twistLimitDegrees=0}
 {
  bone1Name=dyn_ponyTailRoot_01
  bone2Name=head
  swing1LimitDegrees=45
  swing1Motion=Free
  swing2LimitDegrees=45
  swing2Motion=Free
  twistLimitDegrees=45
  twistMotion=Free
 }
 {bone1Name=dyn_ribbon_l_01, bone2Name=head, swing1LimitDegrees=15, swing2LimitDegrees=30, twistLimitDegrees=0}
 {bone1Name=dyn_ribbon_r_01, bone2Name=head, swing1LimitDegrees=30, swing2LimitDegrees=30, twistLimitDegrees=0}
 {bone1Name=dyn_ribbonBase_r, bone2Name=head, swing1LimitDegrees=15, swing2LimitDegrees=15, twistLimitDegrees=0}
 {bone1Name=dyn_ribbonBase_l, bone2Name=head, swing1LimitDegrees=15, swing2LimitDegrees=30, twistLimitDegrees=0}
 {bone1Name=dyn_bangs_01, bone2Name=head, swing1LimitDegrees=15, swing2LimitDegrees=5, twistLimitDegrees=0}
 {bone1Name=dyn_hairMeshBase_r_01, bone2Name=head, swing1LimitDegrees=8, swing2LimitDegrees=8, twistLimitDegrees=0}
 {bone1Name=dyn_hairMeshBase_l_01, bone2Name=head, swing1LimitDegrees=8, swing2LimitDegrees=8, twistLimitDegrees=0}
 {
  bone1Name=head
  bone2Name=neck_02
  swing1LimitDegrees=45
  swing1Motion=Free
  swing2LimitDegrees=45
  swing2Motion=Free
  twistLimitDegrees=45
  twistMotion=Free
 }
 {
  bone1Name=neck_02
  bone2Name=neck_01
  swing1LimitDegrees=45
  swing1Motion=Free
  swing2LimitDegrees=45
  swing2Motion=Free
  twistLimitDegrees=45
  twistMotion=Free
 }
 {
  bone1Name=neck_01
  bone2Name=spine_05
  swing1LimitDegrees=45
  swing1Motion=Free
  swing2LimitDegrees=45
  swing2Motion=Free
  twistLimitDegrees=45
  twistMotion=Free
 }
]
